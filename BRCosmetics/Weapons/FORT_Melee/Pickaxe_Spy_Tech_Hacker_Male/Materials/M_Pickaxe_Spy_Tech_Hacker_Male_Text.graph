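# Translucent, two-sided unlit material that builds holographic scrolling text for the Spy Tech Hacker Male pickaxe.
# Pre-skinned position, projected on YZ and scaled by texture size, drives scanlines and mask panning, with a grayscale texture varying scroll speed.
# The mask's blue channel brightens an edge-to-core color gradient from a pivot-offset sphere mask.
# The result is multiplied by brightened scanlines and fed into MF_FX_EmissiveAdjustments.
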Material /BRCosmetics/Weapons/FORT_Melee/Pickaxe_Spy_Tech_Hacker_Male/Materials/M_Pickaxe_Spy_Tech_Hacker_Male_Text
graph {
 properties {
  bUsedWithMeshDeformer=true
  bUsedWithSkeletalMesh=true
  blendMode=BLEND_Translucent
  pixelDepthOffsetMode=PDOM_Legacy
  shadingModel=MSM_Unlit
  twoSided=true
  usedShadingModels=Unlit
 }
 nodes [
  n1 = VertexInterpolator(VS=PreSkinnedPosition())
  n2 = Abs(None=ScalarParameter() {defaultValue=62, parameterName=Texture Size})
  n3 = ComponentMask(None=Divide(A=@n1.PS, B=Multiply(A=@n2) {constB=-1})) {b=true, g=true}
  n4 = MaterialFunctionCall!(
   ScanLine_Tiling_1=ScalarParameter() {defaultValue=5, parameterName=Holo_ Scanline Tiling 1}
   Scanline_PannerSpeed_1
   ScanLine_Tiling=ScalarParameter() {defaultValue=5, parameterName=Holo_ Scanline Tiling}
   Scanline_PannerSpeed
   UVs=@n3
  ) {materialFunction=/Game/Weapons/WeaponSkins/MaterialFunction/ScanLines}
  n5 = Add(
   A=@n4.Result
   B=ScalarParameter() {defaultValue=0.5, group=Customization_Holographic, parameterName=Holo_ Scanlines Additional Brightness, sortPriority=-197}
  )
  n6 = TextureCoordinate()
  n7 = TextureSample(UVs=Multiply(A=ScalarParameter() {defaultValue=2, parameterName=Text Scroll BreakupTiling}, B=@n6)) {
   samplerType=SAMPLERTYPE_LinearGrayscale
   texture=/BRCosmetics/Weapons/FORT_Melee/Pickaxe_Spy_Tech_Hacker_Male/Textures/T_Pickaxe_Spy_Tech_Hacker_Male_TextSpeed
  }
  n8 = ScalarParameter() {defaultValue=-0.25, parameterName=Text Scroll Breakup}
  n9 = Panner(Coordinate=@n3, Speed=AppendVector(A=Multiply(A=@n7.R, B=@n8), B=Constant())) {speedY=1}
  n10 = TextureSample(UVs=@n9) {
   samplerType=SAMPLERTYPE_Masks
   texture=/BRCosmetics/Weapons/FORT_Melee/Pickaxe_Spy_Tech_Hacker_Male/Textures/T_Pickaxe_Spy_Tech_Hacker_Male_Masks
  }
  n11 = VectorParameter() {group=Emissive, parameterName=EmissiveColor Edges, defaultValue=<rgba 0.0252 0.0369 0.0056 1>}
  n12 = VectorParameter() {group=Emissive, parameterName=EmissiveColor Core, defaultValue=<rgba 0.3864 0.7379 0 1>}
  n13 = VectorParameter() {parameterName=Pre-Skinned Pivot Offset, defaultValue=<rgba 1.136 5.2129 45 0>}
  n14 = Power(
   Base=SphereMask(
    A=@n1.PS
    B=ComponentMask(None=@n13.RGB) {b=true, g=true, r=true}
    Radius=ScalarParameter() {defaultValue=42, parameterName=GlowRadius}
    Hardness=ScalarParameter() {parameterName=GlowHardness}
   ) {hardnessPercent=30}
   Exp=ScalarParameter() {defaultValue=3.75, parameterName=GlowPower}
  )
  n15 = Add(
   A=Multiply(A=@n10.B, B=Constant3Vector() {constant=<rgba 0.6667 1.25 0 0>})
   B=Multiply(A=LinearInterpolate(A=@n11.RGB, B=@n12.RGB, Alpha=@n14), B=ScalarParameter() {defaultValue=0.85, parameterName=Brightness})
  )
  Add(A=@n5, B=@n15)
  n16 = ScalarParameter() {defaultValue=1, group=Customization_Holographic, parameterName=Holo_ Scanlines Thickness, sliderMax=1, sliderMin=0.01, sortPriority=-199}
  Clamp(
   None=OneMinus(None=Power(Base=@n4.Result, Exp=@n16))
   Min=ScalarParameter() {group=Customization_Holographic, parameterName=Holo_ Scanlines Dark Value, sortPriority=-198}
  )
  Constant3Vector()
  Constant3Vector() {constant=<rgba 0 0 1 0>}
  Constant()
  Constant() {r=1}
  MaterialFunctionCall(UVs=@n9, "Rotation Angle (0-1)"=ScalarParameter() {parameterName=Text Rotation}) {materialFunction=&CustomRotator}
  MaterialFunctionCall(UVs=@n6, "Rotation Angle (0-1)"=ScalarParameter() {parameterName=Panner Stripe Rotation}) {materialFunction=&CustomRotator}
  MaterialFunctionCall(Emissive=Multiply(A=@n5, B=@n15)) {materialFunction=&MF_FX_EmissiveAdjustments}
  TextureSample() {
   samplerType=SAMPLERTYPE_Masks
   texture=&T_Pickaxe_Spy_Tech_Hacker_Male_Masks
  }
 ]
}
